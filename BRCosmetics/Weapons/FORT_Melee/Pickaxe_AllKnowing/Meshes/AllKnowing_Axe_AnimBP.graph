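# Animation Blueprint for the AllKnowing pickaxe that makes its top and bottom gem bones float and sway.
# Two chained Anim Dynamics nodes on the input pose give each gem springy linear lag within small limits, with angular limits at zero.
# Additive Modify Bone nodes then layer a procedural bob on both gems.
# The update event computes those offsets from sine and cosine waves of game time at differing frequencies.
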AnimBlueprint /BRCosmetics/Weapons/FORT_Melee/Pickaxe_AllKnowing/Meshes/AllKnowing_Axe_AnimBP
target_skeleton=&/AllKnowing_Axe_Skeleton
graphs [
 AnimGraph {nodes [
   AnimGraphNode_SubInput_0 = Input Pose()
   n1 = Anim Dynamics - Bone: dyn_gem_top(ComponentPose=Local To Component(LocalPose=@AnimGraphNode_SubInput_0), Alpha=1, ExternalForce="0,0,0") {
    angularSpringConstant=50
    bGravityOverrideInSimSpace=true
    bLinearSpring=true
    bOverrideLinearDamping=true
    bUseGravityOverride=true
    linearDampingOverride=0.99
    linearSpringConstant=50
    alphaBoolBlend {blendOption=Linear}
    alphaScaleBiasClamp {interpSpeedDecreasing=3, interpSpeedIncreasing=5}
    boundBone {boneName=dyn_gem_top}
    componentAppliedLinearAccClamp=<100 100 100>
    componentLinearAccScale=<0.05 0.05 0.05>
    componentLinearVelScale=<0.05 0.05 0.05>
    gravityOverride {z=-500}
    physicsBodyDefinitions [{
      boundBone {boneName=dyn_gem_top}
      boxExtents=<90 90 90>
      constraintSetup {
       angularTargetAxis=AxisZ
       angularLimitsMax=<0 0 0>
       angularLimitsMin=<0 0 0>
       angularTarget=<0 0 1>
       linearAxesMax=<0 2 4>
       linearAxesMin=<0 -2 -4>
      }
      localJointOffset=<0 0 -7>
     }]
   }
   n2 = Anim Dynamics - Bone: dyn_gem_bottom(ComponentPose=@n1, Alpha=1, ExternalForce="0,0,0") {
    angularSpringConstant=50
    bGravityOverrideInSimSpace=true
    bLinearSpring=true
    bOverrideLinearDamping=true
    bUseGravityOverride=true
    linearDampingOverride=0.95
    linearSpringConstant=45
    alphaBoolBlend {blendOption=Linear}
    alphaScaleBiasClamp {interpSpeedDecreasing=3, interpSpeedIncreasing=5}
    boundBone {boneName=dyn_gem_bottom}
    componentAppliedLinearAccClamp=<200 200 200>
    componentLinearAccScale=<0.15 0.15 0.15>
    componentLinearVelScale=<0.2 0.2 0.2>
    gravityOverride=<10 10 10>
    physicsBodyDefinitions [{
      boundBone {boneName=dyn_gem_bottom}
      boxExtents=<90 90 90>
      constraintSetup {
       angularTargetAxis=AxisZ
       angularLimitsMax=<0 0 0>
       angularLimitsMin=<0 0 0>
       angularTarget=<0 0 1>
       linearAxesMax=<1.5 1.5 0>
       linearAxesMin=<-1.5 -1.5 0>
      }
      localJointOffset=<0 0 -7>
     }]
   }
   n3 = "Transform (Modify) Bone - Bone: dyn_gem_top"(ComponentPose=@n2, Alpha=1, Translation=Get Top_gem_translate() -> (Top_gem_translate:Vector), Rotation="0,0,0", Scale="1,1,1") {translationMode=BMM_Additive, alphaBoolBlend {blendOption=Linear}, boneToModify {boneName=dyn_gem_top}}
   n4 = Get Bottom_gem_translate() -> (Bottom_gem_translate:Vector)
   Output Pose(Result=Component To Local(ComponentPose="Transform (Modify) Bone - Bone: dyn_gem_bottom"(ComponentPose=@n3, Alpha=1, Translation=@n4, Rotation="0,0,0", Scale="1,1,1") {translationMode=BMM_Additive, alphaBoolBlend {blendOption=Linear}, boneToModify {boneName=dyn_gem_bottom}})) {layerGroup=None}
  ]}
 EventGraph {nodes [
   Event BlueprintUpdateAnimation()
   n1 = GetGameTimeInSeconds()
   n2 = float * float(A=@n1, B=0.5) {is=title K2Node_PromotableOperator}
   n3 = "Sin (Radians)"(A=float * float(A=@n1, B=0.75))
   n4 = "Sin (Radians)"(A=float * float(A=1.5, B=@n1))
   > Set Top_gem_translate(Top_gem_translate:Vector=MakeVector(X="Sin (Radians)"(A=@n2), Y=float + float(A=@n3, B=0.5), Z=float - float(A=@n4, B=0.5))) -> (Output_Get:Vector)
   n5 = float * float(A=@n1, B=0.75)
   n6 = "Cos (Radians)"!(A:double=@n5) -> (ReturnValue:double)
   n7 = "Sin (Radians)"(A=@n5)
   n8 = "Sin (Radians)"(A=float * float(A=1.4, B=@n1))
   > Set Bottom_gem_translate(Bottom_gem_translate:Vector=MakeVector(X=float * float(A=@n7, B=0.5), Y=float * float(A=@n6, B=0.5), Z=float - float(A=@n8, B=0.5))) -> (Output_Get:Vector)
   TryGetPawnOwner()
  ]}
]
variables [
 Bottom_gem_translate {type=Vector}
 Top_gem_translate {type=Vector}
]
